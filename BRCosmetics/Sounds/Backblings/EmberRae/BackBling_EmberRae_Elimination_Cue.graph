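# Sound cue for the EmberRae back bling, named as its elimination sound among reactive cosmetic audio.
# A Random node picks one of five equally weighted wave variations and feeds a Modulator that keeps pitch unchanged.
# The cue uses close-range pet attenuation, the Backblings sound class and the reactive backbling property sheet.
# A concurrency override allows one instance per owner, stopping the oldest when a new one plays.
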SoundCue /BRCosmetics/Sounds/Backblings/EmberRae/BackBling_EmberRae_Elimination_Cue
root=Modulator_0
nodes [
 {id=Modulator_0, children [Random_0], properties {pitchMax=1, pitchMin=1, volumeMax=1}}
 {id=Random_0, children [WavePlayer_3,WavePlayer_4,WavePlayer_5,WavePlayer_6,WavePlayer_7], properties {weights [1,1,1,1,1]}}
 {id=WavePlayer_3, properties {soundWaveAssetPtr=&/BackBling_EmberRae_05}}
 {id=WavePlayer_4, properties {soundWaveAssetPtr=&/BackBling_EmberRae_01}}
 {id=WavePlayer_5, properties {soundWaveAssetPtr=&/BackBling_EmberRae_02}}
 {id=WavePlayer_6, properties {soundWaveAssetPtr=&/BackBling_EmberRae_03}}
 {id=WavePlayer_7, properties {soundWaveAssetPtr=&/BackBling_EmberRae_04}}
]
properties {
 attenuationSettings=&Fort_Pets_Close_Athena_Att
 audioPropertiesSheet=/BRCosmetics/Sounds/Settings/PropertySheets/Reactive/APS_Cosmetic_Reactive_Backbling
 bOverrideConcurrency=true
 maxDistance=1800.0001
 soundClassObject=/Game/Sounds/FortSoundClassesAndModes/Fort_Backblings
 volumeMultiplier=1
 concurrencyOverrides {bLimitToOwner=true, maxCount=1, resolutionRule=StopOldest}
}
